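# Animation Blueprint for the Samurai Blue backpack cosmetic, adding secondary physics motion to its dangling parts.
# Starting from the ref pose, a serial chain of Anim Dynamics nodes simulates head, bow, bow bottom, tassles, sword and both leg bones with damped angular springs and tight angular limits.
# A final Transform Bone node additively shifts backpack_root along Y before output.
# The event graph has no logic and there are no variables.
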AnimBlueprint /BRCosmetics/Accessories/FORT_Backpacks/F_MED_SamuraiBlue_Backpack/Meshes/F_MED_Samurai_Blue_Backpack_AnimBP
target_skeleton=&/F_MED_Samurai_Blue_Backpack_Skeleton
graphs [
 AnimGraph {nodes [
   n1 = Anim Dynamics - Bone: dyn_head!(ComponentPose=Local To Component(LocalPose=Local Space Ref Pose()), ExternalForce="0,0,0") -> (Pose) {
    angularDampingOverride=0.99
    angularSpringConstant=5000
    bAngularSpring=true
    bGravityOverrideInSimSpace=true
    bOverrideAngularDamping=true
    gravityScale=0.35
    numSolverIterationsPostUpdate=4
    numSolverIterationsPreUpdate=16
    alphaBoolBlend {blendOption=Linear}
    boundBone {boneName=dyn_head}
    physicsBodyDefinitions [{
      sphereCollisionRadius=0
      boundBone {boneName=dyn_head}
      boxExtents=<40 40 30>
      constraintSetup {angularTargetAxis=AxisZ, twistAxis=AxisZ, angularLimitsMax=<5 20 20>, angularLimitsMin=<-5 -20 -15>, angularTarget=<0 0 1>}
      localJointOffset=<0 0 -3>
     }]
   }
   n2 = Anim Dynamics - Bone: dyn_bow!(ComponentPose=@n1, ExternalForce="0,0,0") -> (Pose) {
    angularDampingOverride=0.99
    angularSpringConstant=5000
    bAngularSpring=true
    bGravityOverrideInSimSpace=true
    bOverrideAngularDamping=true
    gravityScale=0.35
    numSolverIterationsPostUpdate=4
    numSolverIterationsPreUpdate=16
    alphaBoolBlend {blendOption=Linear}
    boundBone {boneName=dyn_bow}
    physicsBodyDefinitions [{
      sphereCollisionRadius=0
      boundBone {boneName=dyn_bow}
      boxExtents=<10 10 5>
      constraintSetup {angularTargetAxis=AxisZ, twistAxis=AxisZ, angularLimitsMax=<5 20 20>, angularLimitsMin=<-5 -20 -15>, angularTarget=<0 0 1>}
      localJointOffset=<0 0 -1>
     }]
   }
   n3 = Anim Dynamics - Bone: dyn_bow_bott!(ComponentPose=@n2, ExternalForce="0,0,0") -> (Pose) {
    angularDampingOverride=0.99
    angularSpringConstant=5000
    bAngularSpring=true
    bGravityOverrideInSimSpace=true
    bOverrideAngularDamping=true
    numSolverIterationsPostUpdate=4
    numSolverIterationsPreUpdate=16
    alphaBoolBlend {blendOption=Linear}
    boundBone {boneName=dyn_bow_bott}
    physicsBodyDefinitions [{
      sphereCollisionRadius=0
      boundBone {boneName=dyn_bow_bott}
      boxExtents=<10 10 5>
      constraintSetup {angularTargetAxis=AxisZ, twistAxis=AxisZ, angularLimitsMax=<5 20 20>, angularLimitsMin=<-5 -20 -15>, angularTarget=<0 0 1>}
      localJointOffset=<0 0 -1>
     }]
   }
   n4 = Anim Dynamics - Bone: dyn_tassles!(ComponentPose=@n3, ExternalForce="0,0,0") -> (Pose) {
    angularDampingOverride=0.99
    angularSpringConstant=5000
    bAngularSpring=true
    bGravityOverrideInSimSpace=true
    bOverrideAngularDamping=true
    numSolverIterationsPostUpdate=4
    numSolverIterationsPreUpdate=16
    alphaBoolBlend {blendOption=Linear}
    boundBone {boneName=dyn_tassles}
    physicsBodyDefinitions [{
      sphereCollisionRadius=0
      boundBone {boneName=dyn_tassles}
      boxExtents=<10 10 20>
      constraintSetup {angularTargetAxis=AxisZ, twistAxis=AxisZ, angularLimitsMax=<5 20 20>, angularLimitsMin=<-5 -20 -15>, angularTarget=<0 0 1>}
      localJointOffset=<0 0 -3>
     }]
   }
   n5 = Anim Dynamics - Bone: dyn_sword!(ComponentPose=@n4, ExternalForce="0,0,0") -> (Pose) {
    angularDampingOverride=0.99
    angularSpringConstant=3000
    bAngularSpring=true
    bGravityOverrideInSimSpace=true
    bOverrideAngularDamping=true
    gravityScale=0.35
    numSolverIterationsPostUpdate=4
    numSolverIterationsPreUpdate=16
    alphaBoolBlend {blendOption=Linear}
    boundBone {boneName=dyn_sword}
    physicsBodyDefinitions [{
      sphereCollisionRadius=0
      boundBone {boneName=dyn_sword}
      boxExtents=<60 5 5>
      constraintSetup {angularTargetAxis=AxisZ, twistAxis=AxisZ, angularLimitsMax=<3 3 10>, angularLimitsMin=<-3 -15 -10>, angularTarget=<0 0 1>}
      localJointOffset=<0 0 -10>
     }]
   }
   n6 = Anim Dynamics - Bone: dyn_leg_l!(ComponentPose=@n5, ExternalForce="0,0,0") -> (Pose) {
    angularDampingOverride=0.99
    angularSpringConstant=5000
    bAngularSpring=true
    bGravityOverrideInSimSpace=true
    bOverrideAngularDamping=true
    numSolverIterationsPostUpdate=4
    numSolverIterationsPreUpdate=16
    alphaBoolBlend {blendOption=Linear}
    boundBone {boneName=dyn_leg_l}
    physicsBodyDefinitions [{
      sphereCollisionRadius=0
      boundBone {boneName=dyn_leg_l}
      boxExtents=<8 20 8>
      constraintSetup {angularTargetAxis=AxisZ, twistAxis=AxisZ, angularLimitsMax=<5 20 20>, angularLimitsMin=<-5 -20 -15>, angularTarget=<0 0 1>}
      localJointOffset=<0 0 -3>
     }]
   }
   n7 = Anim Dynamics - Bone: dyn_leg_r!(ComponentPose=@n6, ExternalForce="0,0,0") -> (Pose) {
    angularDampingOverride=0.99
    angularSpringConstant=5000
    bAngularSpring=true
    bGravityOverrideInSimSpace=true
    bOverrideAngularDamping=true
    numSolverIterationsPostUpdate=4
    numSolverIterationsPreUpdate=16
    alphaBoolBlend {blendOption=Linear}
    boundBone {boneName=dyn_leg_r}
    physicsBodyDefinitions [{
      sphereCollisionRadius=0
      boundBone {boneName=dyn_leg_r}
      boxExtents=<8 20 8>
      constraintSetup {angularTargetAxis=AxisZ, twistAxis=AxisZ, angularLimitsMax=<5 20 20>, angularLimitsMin=<-5 -20 -15>, angularTarget=<0 0 1>}
      localJointOffset=<0 0 -3>
     }]
   }
   n8 = "Transform (Modify) Bone - Bone: backpack_root"(ComponentPose=@n7, Alpha=1, Translation="0,-5,0", Rotation="0,0,0", Scale="1,1,1") {translationMode=BMM_Additive, alphaBoolBlend {blendOption=Linear}, boneToModify {boneName=backpack_root}, translation {y=-5}}
   Output Pose(Result=Component To Local(ComponentPose=@n8)) {layerGroup=None}
  ]}
 EventGraph {nodes [Event BlueprintUpdateAnimation()]}
]
variables []
